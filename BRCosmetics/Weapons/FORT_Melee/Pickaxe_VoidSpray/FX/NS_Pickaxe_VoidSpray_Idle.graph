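# Idle glow effect for the VoidSpray pickaxe, drawing persistent blue sprites at sockets on the weapon.
# LightAndGlow burst-spawns two differently sized sprite groups tagged for its two renderers, while LightAndGlow003 spawns a mirrored pair following the second socket's rotation.
# Shared system values for fade-in, waveform flicker and TOD tint scale both emitters' color.
# Front-end modifiers give lobby and gameplay separate alpha and, for LightAndGlow, size.
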NiagaraSystem /BRCosmetics/Weapons/FORT_Melee/Pickaxe_VoidSpray/FX/NS_Pickaxe_VoidSpray_Idle
dependencies {
 usedDataInterfaces [&NiagaraDataInterfaceSocketReader, &NiagaraDataInterfaceColorCurve, &NiagaraDataInterfaceCurve]
 usedDynamicInputs [
  &ColorFromCurve
  &Waveform
  &FloatFromCurve
 ]
 usedModules [
  &self:SystemSpawnScript.SystemScriptSource.SystemScriptGraph.NiagaraNodeAssignment_1.SetVariables_#1
  &SystemState
  &self:SystemSpawnScript.SystemScriptSource.SystemScriptGraph.NiagaraNodeAssignment_0.SetVariables_#2
 ]
}
emitters {
 LightAndGlow {
  emitterUpdateScript [
   &EmitterState(
    Inactive Response=::NewEnumerator1
    Life Cycle Mode=::NewEnumerator1
    Loop Behavior=::NewEnumerator1
    Loop Duration Mode=::NewEnumerator1
    UseLoopDelay=false
   )
   &SpawnBurst_Instantaneous(Use Spawn Probability=false, UseLoopCountLimit=false)
   &SpawnBurst_Instantaneous(Spawn Group=1, Use Spawn Probability=false, UseLoopCountLimit=false) {module=SpawnBurst_Instantaneous002}
  ]
  particleSpawnScript [
   /SetVariables(Particles.VisibilityTag=&SpawnGroup()) {isSetParameters=true}
   &InitializeParticle(
    Color=&MakeLinearColorFromVectorAndFloat(
     "Vector (RGB)"=&MakeVectorFromLinearColorRGB(Color=<rgba 0.0289 0.4099 0.7217 1>)
     "Float (Alpha)"=&FN_FrontEndFloatModifier(Gameplay=0.12, Front End=0.09)
    )
    Sprite Size=&MaskVector2dBySpawnGroup(
     Masked Vector2D=<16.92 72>
     Passthrough Vector2D=&FN_FrontEndVector2Modifier(Front End:Vector2f=<7 77>, Gameplay:Vector2f=<11 77>)
     Spawn Group Mask=0
    )
    Sprite Size Mode=::NewEnumerator4
    UsePositionOffset=false
    Write Material Random=true
   )
   &SocketLocation(SamplePosOffset=false, Socket Reader=$System.Socket Reader, Socket Sampling Mode=::NewEnumerator3, VelocitySpeedLimit=false)
   &CameraOffset(Camera Offset Amount=5)
   &SpriteFacingAndAlignment(
    Sprite Alignment=<0 0 1>
    Sprite Facing=&MaskVectorBySpawnGroup(Masked Vector=<0 1 0>, Passthrough Vector=<1 0 0>)
    WriteAlignment=true
    WriteFacing=true
   )
  ]
  particleUpdateScript [
   &ParticleState(Kill Particles When Lifetime Has Elapsed=false)
   &ScaleColor(
    Scale Alpha=$System.FadeIn
    Scale RGB=&VectorFromFloat(Value=$System.WaveformFloat)
    ScaleA=true
    ScaleRGB=true
   )
   &ScaleColor(Color Scale:LinearColor=$System.TOD, Scale Mode=3) {module=ScaleColor001}
  ]
  rendererClasses [&NiagaraSpriteRendererProperties]
  renderers [
   {rendererClass=&NiagaraSpriteRendererProperties}
   {rendererClass=&NiagaraSpriteRendererProperties, rendererIndex=1}
  ]
 }
 LightAndGlow003 {
  emitterUpdateScript [
   &EmitterState(
    Inactive Response=::NewEnumerator1
    Life Cycle Mode=::NewEnumerator1
    Loop Behavior=::NewEnumerator1
    Loop Duration Mode=::NewEnumerator1
    UseLoopDelay=false
   )
   &SpawnBurst_Instantaneous(Spawn Count=2, Use Spawn Probability=false, UseLoopCountLimit=false)
  ]
  particleSpawnScript [
   &FN_SetSpawnGroup(Number of Spawn Groups=2)
   &InitializeParticle(
    Color=&MakeLinearColorFromVectorAndFloat(
     "Vector (RGB)"=&MakeVectorFromLinearColorRGB(Color=<rgba 0.0289 0.4099 0.7217 1>)
     "Float (Alpha)"=&FN_FrontEndFloatModifier(Gameplay=0.12, Front End=0.09)
    )
    Sprite Size=<5 5>
    Sprite Size Mode=::NewEnumerator4
    UsePositionOffset=false
    Write Material Random=true
   )
   &SocketLocation(
    Orientation Sampling=::NewEnumerator1
    SamplePosOffset=true
    Sampled Position Offset=&Multiply_VectorByFloat(Vector=<-2 0 0>, Float=&MaskFloatBySpawnGroup(Masked Float=-1, Passthrough Float=1))
    Socket Index=1
    Socket Reader=$System.Socket Reader
    Socket Sampling Mode=::NewEnumerator3
    VelocitySpeedLimit=false
   )
   &CameraOffset()
   &SpriteFacingAndAlignment(
    Sprite Alignment=&RotateVectorByQuaternion(Vector To Rotate=<0 0 1>, Quaternion=$Output.SocketLocation.SampledRotation)
    Sprite Facing=&Multiply_VectorByFloat(Vector=<0 1 0>, Float=&MaskFloatBySpawnGroup(Masked Float=-1, Passthrough Float=1))
    WriteAlignment=true
    WriteFacing=true
   )
  ]
  particleUpdateScript [
   &ParticleState(Kill Particles When Lifetime Has Elapsed=false)
   &ScaleColor(
    Scale Alpha=$System.FadeIn
    Scale RGB=&VectorFromFloat(Value=$System.WaveformFloat)
    ScaleA=true
    ScaleRGB=true
   )
   &ScaleColor(Color Scale:LinearColor=$System.TOD, Scale Mode=3) {module=ScaleColor001}
  ]
  rendererClasses [&NiagaraSpriteRendererProperties]
  renderers [{rendererClass=&NiagaraSpriteRendererProperties}]
 }
}
summary {
 systemName=NS_Pickaxe_VoidSpray_Idle
 emitters [
  {emitterName=LightAndGlow, rendererClasses [&NiagaraSpriteRendererProperties]}
  {emitterName=LightAndGlow003, rendererClasses [&NiagaraSpriteRendererProperties]}
 ]
 userVariables [
  User.bIsFrontEnd {type=NiagaraBool}
  User.bIsFrontEndPreview {type=NiagaraBool}
 ]
}
