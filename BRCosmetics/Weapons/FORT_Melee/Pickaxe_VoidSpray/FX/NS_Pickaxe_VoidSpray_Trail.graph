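# Niagara trail effect for the VoidSpray pickaxe, drawing short-lived blue swing trails with circle particles.
# Trail_Main and Trail_Edge build multi-renderer ribbons from shared AnimTrail modules, with Trail_Main computing separate dark, emissive and glow colors.
# Particle_Circles spawns sprites offset from a weapon socket, oriented by socket velocity and rotation, with alpha adjusted for front end.
# System-level time-of-day color tints all emitters, while fade values drive trail alpha and circle spawn rate.
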NiagaraSystem /BRCosmetics/Weapons/FORT_Melee/Pickaxe_VoidSpray/FX/NS_Pickaxe_VoidSpray_Trail
dependencies {
 usedDataInterfaces [&NiagaraDataInterfaceColorCurve, &NiagaraDataInterfaceCurve, &NiagaraDataInterfaceSocketReader]
 usedDynamicInputs [&FNTimeOfDayColor, &MergeHelper_FloatSwitch]
 usedModules [
  &self:SystemSpawnScript.SystemScriptSource.SystemScriptGraph.NiagaraNodeAssignment_1.SetVariables_#1
  &SystemState
  &self:SystemSpawnScript.SystemScriptSource.SystemScriptGraph.NiagaraNodeAssignment_4.SetVariables_#2
  &self:SystemSpawnScript.SystemScriptSource.SystemScriptGraph.NiagaraNodeAssignment_7.SetVariables_#3
 ]
}
emitters {
 Particle_Circles {
  emitterUpdateScript [
   &EmitterState()
   &SpawnRate(Spawn Probability=0.8, SpawnRate=&Multiply_Float(A=55, B=$System.FadeOut), Use Spawn Probability=true)
  ]
  particleSpawnScript [
   &InitializeParticle(
    Color=&MakeLinearColorFromVectorAndFloat("Vector (RGB)"=<1 1 1>, "Float (Alpha)"=&FN_FrontEndFloatModifier(Gameplay=1.5, Front End=0.7))
    Lifetime Max=0.32
    Lifetime Min=0.22
    Lifetime Mode=::NewEnumerator1
    Sprite Rotation Mode=::NewEnumerator2
    Sprite Size Mode=::NewEnumerator1
    Uniform Sprite Size Max=35
    Uniform Sprite Size Min=25
    UsePositionOffset=false
    Write Material Random=true
   )
   &CameraOffset(Camera Offset Amount=15)
   &SocketLocation(
    SamplePosOffset=true
    Sampled Position Offset=&MakeVector(
     X=0
     Y=&RandomRangeFloat(Minimum=5, Maximum=12)
     Z=&RandomRangeFloat(Minimum=-10, Maximum=-50)
    )
    Socket Index=1
    Socket Reader=$System.Socket Reader
    Socket Sampling Mode=::NewEnumerator3
    VelocitySpeedLimit=false
   )
   &SpriteFacingAndAlignment(
    Sprite Alignment=&NormalizeVector(Vector To Normalize=$Output.SocketLocation.SampledVelocity, Safe Normalize=true, Safe Fallback Threshold=0, Safe Fallback Vector=<0 0 0.001>)
    Sprite Facing=/Niagara/DynamicInputs/Vectors/CrossProduct(
     Vector A:Vector3f=&RotateVectorByQuaternion(Vector To Rotate=<0 0 1>, Quaternion=$Output.SocketLocation.SampledRotation)
     Vector B:Vector3f=$Output.SocketLocation.SampledVelocity
     Normalize Incoming Vectors:NiagaraBool=true
    )
    WriteAlignment=true
    WriteFacing=true
   )
  ]
  particleUpdateScript [
   &ParticleState()
   &ScaleSpriteSize(Uniform Curve Sprite Scale=DataInterface() {propertyValues {Curve {
       defaultValue=0
       keys [
        {arriveTangent=1.4266, interpMode=RCIM_Cubic, leaveTangent=1.4266, tangentMode=RCTM_User, time=0, value=0.75}
        {arriveTangent=0.3797, interpMode=RCIM_Cubic, leaveTangent=0.3797, tangentMode=RCTM_User, time=0.3494, value=0.9227}
        {arriveTangent=0.0224, interpMode=RCIM_Cubic, leaveTangent=0.0224, tangentMode=RCTM_User, time=1, value=1}
       ]
      }}})
   &ScaleColor(
    Linear Color Curve=DataInterface() {propertyValues {
      AlphaCurve {keys [<at 0.0072 0>,<at 0.1244 0.1477>,<at 0.268 0.8>,<at 0.4181 0.2>,<at 1 0>]}
      BlueCurve {keys [<at 0.1929 1>]}
      GreenCurve {keys [<at 0.1929 0.7758>]}
      RedCurve {keys [<at 0.1929 0.2623>]}
     }}
    Scale Mode=::NewEnumerator2
   )
   &ScaleColor(Color Scale:LinearColor=$System.TODColor, Scale Mode=3) {module=ScaleColor001}
   &DynamicMaterialParameters(
    Index 0 Param 1=&FloatFromCurve(
     FloatCurve=DataInterface() {propertyValues {Curve {
        defaultValue=1
        keys [
         {arriveTangent=2.0818, interpMode=RCIM_Cubic, leaveTangent=2.0818, tangentMode=RCTM_User, time=0, value=0}
         {arriveTangent=2.1952, interpMode=RCIM_Cubic, leaveTangent=2.1952, tangentMode=RCTM_User, time=1, value=1}
        ]
       }}}
     Scale Curve=&RandomRangeFloat(Minimum=0, Maximum=0.2)
    )
    Module.Index0Param0Write=true
    Module.Index0Param1Write=true
    Module.Index0Param2Write=false
    Module.Index0Param3Write=false
   )
  ]
  rendererClasses [&NiagaraSpriteRendererProperties]
  renderers [{rendererClass=&NiagaraSpriteRendererProperties}]
 }
 Trail_Edge {
  emitterUpdateScript [
   &EmitterState()
   &FN_SpawnAnimTrail(AnimTrailWidthMode=::FromSecond)
  ]
  particleSpawnScript [
   &FN_InitializeAnimTrailBoneIndecies()
   &FN_InitializeAnimTrail(
    AnimTrailWidth=0.075
    Color=&MakeLinearColorFromVectorAndFloat(
     "Vector (RGB)"=&MakeVectorFromLinearColorRGB(Color=<rgba 1 1 1 1>)
     "Float (Alpha)"=&Multiply_Float(A=1, B=$System.FadeIn)
    )
    Lifetime=0.11
   )
   /SetVariables(Particles.ScreenTrailWidth=3) {isSetParameters=true}
  ]
  particleUpdateScript [
   &FN_UpdateAnimTrail() {module=FN_UpdateAnimTrail001}
   &ScaleColor(
    Scale Mode=::NewEnumerator1
    Scale RGBA=&Multiply_Vector4(
     A=&MakeVector4_FromLinearColor(LinearColor=$System.TODColor)
     B=&MakeVector4FromVectorAndFloat("Vector (RGB)"=<1 1 1>, "Float (Alpha)"=$System.FadeOut)
    )
    ScaleRGBA=true
   )
   &ScaleColor(
    Linear Color Curve=DataInterface() {propertyValues {
      AlphaCurve {keys [<at 0 1>]}
      BlueCurve {keys [<at 0 2>,<at 0.5285 1>]}
      GreenCurve {keys [<at 0 0.9068>,<at 0.5285 0.2911>]}
      RedCurve {keys [<at 0 0>,<at 0.5285 0>]}
     }}
    Scale Mode=2
   ) {module=ScaleColor002}
  ]
  rendererClasses [&NiagaraRibbonRendererProperties]
  renderers [
   {rendererClass=&NiagaraRibbonRendererProperties}
   {rendererClass=&NiagaraRibbonRendererProperties, rendererIndex=1}
  ]
 }
 Trail_Main {
  emitterUpdateScript [
   &EmitterState()
   &FN_SpawnAnimTrail()
  ]
  particleSpawnScript [
   &FN_InitializeAnimTrailBoneIndecies()
   &FN_InitializeAnimTrail(
    Color=&MakeLinearColorFromVectorAndFloat(
     "Vector (RGB)"=&MakeVectorFromLinearColorRGB(Color=<rgba 1 1 1 1>)
     "Float (Alpha)"=&Multiply_Float(A=1, B=$System.FadeIn)
    )
    Lifetime=0.12
   )
  ]
  particleUpdateScript [
   &FN_UpdateAnimTrail() {module=FN_UpdateAnimTrail001}
   &ScaleColor(
    Scale Mode=::NewEnumerator1
    Scale RGBA=&Multiply_Vector4(
     A=&MakeVector4_FromLinearColor(LinearColor=$System.TODColor)
     B=&MakeVector4FromVectorAndFloat("Vector (RGB)"=<1 1 1>, "Float (Alpha)"=$System.FadeOut)
    )
    ScaleRGBA=true
   )
   /SetVariables(
    Particles.DarkColor:LinearColor=&MultiplyLinearColor(
     Linear Color=$Particles.Color
     Scale Factor=&MakeVector4_FromLinearColor(LinearColor=&ColorFromCurve(ColorCurve=DataInterface() {propertyValues {
         AlphaCurve {keys [<at 0.0148 2>,<at 0.1357 1 RCIM_Cubic>]}
         BlueCurve {keys [<at 0 0.287 RCIM_Cubic>,<at 0.3606 0.1053>]}
         GreenCurve {keys [<at 0 0.1116 RCIM_Cubic>,<at 0.3606 0.0349>]}
         RedCurve {keys [<at 0 0.0108 RCIM_Cubic>,<at 0.3606 0.005>]}
        }}))
    )
    Particles.EmissiveColor:LinearColor=&MultiplyLinearColor(
     Linear Color=$Particles.Color
     Scale Factor=&MakeVector4_FromLinearColor(LinearColor=&ColorFromCurve(ColorCurve=DataInterface() {propertyValues {
         AlphaCurve {keys [<at 0.1666 40>,<at 0.3859 25>,<at 0.6987 10>]}
         BlueCurve {keys [<at 0 1.5>,<at 0.3845 1>]}
         GreenCurve {keys [<at 0 0.7293>,<at 0.3845 0.4905>]}
         RedCurve {keys [<at 0 0.2087>,<at 0.3845 0.1875>]}
        }}))
    )
    Particles.GlowColor=&MultiplyLinearColor(
     Linear Color=$Particles.Color
     Scale Factor=&MakeVector4_FromLinearColor(LinearColor=<rgba 0.0496 0.2069 0.6609 4>)
    )
   ) {isSetParameters=true}
  ]
  rendererClasses [&NiagaraRibbonRendererProperties]
  renderers [
   {rendererClass=&NiagaraRibbonRendererProperties}
   {rendererClass=&NiagaraRibbonRendererProperties, rendererIndex=1}
   {rendererClass=&NiagaraRibbonRendererProperties, rendererIndex=2}
   {rendererClass=&NiagaraRibbonRendererProperties, rendererIndex=3}
  ]
 }
}
summary {
 systemName=NS_Pickaxe_VoidSpray_Trail
 emitters [
  {emitterName=Trail_Main, rendererClasses [&NiagaraRibbonRendererProperties]}
  {emitterName=Trail_Edge, rendererClasses [&NiagaraRibbonRendererProperties]}
  {emitterName=Particle_Circles, rendererClasses [&NiagaraSpriteRendererProperties]}
 ]
 userVariables [
  User.bIsFrontEnd {type=NiagaraBool}
  User.bIsFrontEndPreview {type=NiagaraBool}
  User.RibbonIDOffset {type=NiagaraInt32}
  User.SwingSectionIndex {type=NiagaraInt32}
 ]
}
