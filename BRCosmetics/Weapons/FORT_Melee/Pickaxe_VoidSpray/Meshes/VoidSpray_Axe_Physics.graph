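# Physics asset for the VoidSpray pickaxe mesh, defining capsule collision and joints for a chain attached to the two-handed handle.
# Bodies link in series from the handle through base_chain_01 to dyn_chain_01, 02, 03 and 05, with no body for a 04 bone.
# The first dynamic link swings mostly in one plane, later links swing more freely, and twist is nearly locked along the dynamic chain.
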PhysicsAsset /BRCosmetics/Weapons/FORT_Melee/Pickaxe_VoidSpray/Meshes/VoidSpray_Axe_Physics
bodies {
 Melee_TwoHanded_Handle [{shapeType=Capsule, length=7, radius=2.5, center=<-0.3536 1.3493 -7.0723>, rotation=<rot -50 -90 0>}]
 base_chain_01 [{shapeType=Capsule, length=0.505, radius=0.505, shapeName=base_chain_01_capsule, rotation=<rot 90 0 0>}]
 dyn_chain_01 [{shapeType=Capsule, length=0.5, radius=1, shapeName=dyn_chain_01_capsule, center=<6 0 0>, rotation=<rot 90 0 0>}]
 dyn_chain_02 [{shapeType=Capsule, length=2.5, radius=1, shapeName=dyn_chain_02_capsule, center=<2 0 0>, rotation=<rot 90 0 0>}]
 dyn_chain_03 [{shapeType=Capsule, length=2, radius=1, shapeName=dyn_chain_03_capsule, center=<2 0 0>, rotation=<rot 90 0 0>}]
 dyn_chain_05 [{shapeType=Capsule, length=1, radius=2.5, shapeName=dyn_chain_05_capsule, center=<3 0 0>, rotation=<rot 0 0 90>}]
}
constraints [
 {bone1Name=base_chain_01, bone2Name=Melee_TwoHanded_Handle, swing1LimitDegrees=45, swing2LimitDegrees=45, twistLimitDegrees=45}
 {bone1Name=dyn_chain_01, bone2Name=base_chain_01, swing1LimitDegrees=75, swing2LimitDegrees=1, twistLimitDegrees=1}
 {bone1Name=dyn_chain_02, bone2Name=dyn_chain_01, swing1LimitDegrees=45, swing2LimitDegrees=30, twistLimitDegrees=1}
 {bone1Name=dyn_chain_03, bone2Name=dyn_chain_02, swing1LimitDegrees=45, swing2LimitDegrees=45, twistLimitDegrees=1}
 {bone1Name=dyn_chain_05, bone2Name=dyn_chain_03, swing1LimitDegrees=45, swing2LimitDegrees=45, twistLimitDegrees=1}
]
