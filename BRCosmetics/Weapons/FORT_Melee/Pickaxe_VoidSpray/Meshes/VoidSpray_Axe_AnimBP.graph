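# Anim Blueprint for the VoidSpray pickaxe cosmetic that adds secondary physics motion to its chain bones.
# In the front end, it pitches base_chain_01 by an amount chosen by bIsOnPawn, then runs four partial-alpha AnimDynamics nodes down the chain.
# Otherwise, a RigidBody node simulates the cached input pose relative to the handle bone.
# Other physics nodes are unconnected, and the variables set on customization load are unused by the AnimGraph.
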AnimBlueprint /BRCosmetics/Weapons/FORT_Melee/Pickaxe_VoidSpray/Meshes/VoidSpray_Axe_AnimBP
target_skeleton=&/VoidSpray_Axe_Skeleton
graphs [
 AnimGraph {nodes [
   n1 = Anim Dynamics - Bone: dyn_chain_01(Alpha=1, ExternalForce="0,0,0") {
    angularDampingOverride=1
    angularSpringConstant=200
    bAngularSpring=true
    bGravityOverrideInSimSpace=true
    bOverrideAngularDamping=true
    simulationSpace=World
    boundBone {boneName=dyn_chain_01}
    componentLinearAccScale=<1 1 1>
    componentLinearVelScale=<1 1 1>
    gravityOverride {z=-2000}
    physicsBodyDefinitions [{
      boundBone {boneName=dyn_chain_01}
      boxExtents=<50 50 50>
      constraintSetup {angularLimitsMax=<15 15 15>, angularLimitsMin=<-15 -15 -15>, angularTarget=<1 0 0>}
      localJointOffset=<4 0 0>
     }]
    relativeSpaceBone {boneName=b_AR_Root}
   }
   n2 = Anim Dynamics - Bone: dyn_chain_02(ComponentPose=@n1, Alpha=1, ExternalForce="0,0,0") {
    angularDampingOverride=1
    angularSpringConstant=200
    bAngularSpring=true
    bGravityOverrideInSimSpace=true
    bOverrideAngularDamping=true
    simulationSpace=World
    boundBone {boneName=dyn_chain_02}
    componentLinearAccScale=<1 1 1>
    componentLinearVelScale=<1 1 1>
    gravityOverride {z=-2000}
    physicsBodyDefinitions [{
      boundBone {boneName=dyn_chain_02}
      boxExtents=<50 50 50>
      constraintSetup {angularLimitsMax=<15 15 15>, angularLimitsMin=<-15 -15 -15>, angularTarget=<1 0 0>}
      localJointOffset=<4 0 0>
     }]
    relativeSpaceBone {boneName=b_AR_Root}
   }
   n3 = Anim Dynamics - Bone: dyn_chain_03(ComponentPose=@n2, Alpha=1, ExternalForce="0,0,0") {
    angularDampingOverride=1
    angularSpringConstant=200
    bAngularSpring=true
    bGravityOverrideInSimSpace=true
    bOverrideAngularDamping=true
    simulationSpace=World
    boundBone {boneName=dyn_chain_03}
    componentLinearAccScale=<1 1 1>
    componentLinearVelScale=<1 1 1>
    gravityOverride {z=-2000}
    physicsBodyDefinitions [{
      boundBone {boneName=dyn_chain_03}
      boxExtents=<50 50 50>
      constraintSetup {angularLimitsMax=<15 15 15>, angularLimitsMin=<-15 -15 -15>, angularTarget=<1 0 0>}
      localJointOffset=<4 0 0>
     }]
    relativeSpaceBone {boneName=b_AR_Root}
   }
   Anim Dynamics - Bone: dyn_chain_05(ComponentPose=@n3, Alpha=1, ExternalForce="0,0,0") {
    angularDampingOverride=1
    angularSpringConstant=200
    bAngularSpring=true
    bGravityOverrideInSimSpace=true
    bOverrideAngularDamping=true
    simulationSpace=World
    boundBone {boneName=dyn_chain_05}
    componentLinearAccScale=<1 1 1>
    componentLinearVelScale=<1 1 1>
    gravityOverride {z=-2000}
    physicsBodyDefinitions [{
      boundBone {boneName=dyn_chain_05}
      boxExtents=<50 50 50>
      constraintSetup {angularLimitsMax=<15 15 15>, angularLimitsMin=<-15 -15 -15>, angularTarget=<1 0 0>}
      localJointOffset=<4 0 0>
     }]
    relativeSpaceBone {boneName=b_AR_Root}
   }
   n4 = Anim Dynamics - Bone: dyn_chain_01(Alpha=1, ExternalForce="0,0,0") {
    angularDampingOverride=0.7
    angularSpringConstant=25
    bAngularSpring=true
    bOverrideAngularDamping=true
    bUseGravityOverride=true
    simulationSpace=Actor
    boundBone {boneName=dyn_chain_01}
    componentAppliedLinearAccClamp=<1000 1000 1000>
    componentLinearAccScale=<1 1 1>
    componentLinearVelScale=<1 1 1>
    gravityOverride {z=-980}
    physicsBodyDefinitions [{
      boundBone {boneName=dyn_chain_01}
      boxExtents=<10 10 10>
      constraintSetup {angularLimitsMax=<1 1 22>, angularLimitsMin=<-1 -1 -100>, angularTarget=<1 0 0>}
      localJointOffset=<1 0 0>
     }]
    simSpaceSettings {maxAngularAcceleration=75, maxAngularVelocity=75, simSpaceAngularAlpha=1}
   }
   n5 = Anim Dynamics - Bone: dyn_chain_02(ComponentPose=@n4, Alpha=1, ExternalForce="0,0,0") {
    angularDampingOverride=0.999
    angularSpringConstant=25
    bAngularSpring=true
    bOverrideAngularDamping=true
    bUseGravityOverride=true
    simulationSpace=Actor
    boundBone {boneName=dyn_chain_02}
    componentAppliedLinearAccClamp=<1000 1000 1000>
    componentLinearAccScale=<1 1 1>
    componentLinearVelScale=<1 1 1>
    gravityOverride {z=-900}
    physicsBodyDefinitions [{
      boundBone {boneName=dyn_chain_02}
      boxExtents=<10 10 10>
      constraintSetup {angularLimitsMax=<1 30 45>, angularLimitsMin=<-1 -30 -35>, angularTarget=<1 0 0>}
      localJointOffset=<1 0 0>
     }]
    simSpaceSettings {maxAngularAcceleration=75, maxAngularVelocity=75, simSpaceAngularAlpha=1}
   }
   n6 = Anim Dynamics - Bone: dyn_chain_03(ComponentPose=@n5, Alpha=1, ExternalForce="0,0,0") {
    angularDampingOverride=0.999
    angularSpringConstant=25
    bAngularSpring=true
    bOverrideAngularDamping=true
    bUseGravityOverride=true
    simulationSpace=Actor
    boundBone {boneName=dyn_chain_03}
    componentAppliedLinearAccClamp=<1000 1000 1000>
    componentLinearAccScale=<1 1 1>
    componentLinearVelScale=<1 1 1>
    gravityOverride {z=-900}
    physicsBodyDefinitions [{
      boundBone {boneName=dyn_chain_03}
      boxExtents=<10 10 10>
      constraintSetup {angularLimitsMax=<1 45 45>, angularLimitsMin=<-1 -45 -45>, angularTarget=<1 0 0>}
      localJointOffset=<1 0 0>
     }]
    simSpaceSettings {maxAngularAcceleration=75, maxAngularVelocity=75, simSpaceAngularAlpha=1}
   }
   Anim Dynamics - Bone: dyn_chain_05(ComponentPose=@n6, Alpha=1, ExternalForce="0,0,0") {
    angularDampingOverride=0.999
    angularSpringConstant=25
    bAngularSpring=true
    bOverrideAngularDamping=true
    bUseGravityOverride=true
    simulationSpace=Actor
    boundBone {boneName=dyn_chain_05}
    componentAppliedLinearAccClamp=<1000 1000 1000>
    componentLinearAccScale=<1 1 1>
    componentLinearVelScale=<1 1 1>
    gravityOverride {z=-900}
    physicsBodyDefinitions [{
      boundBone {boneName=dyn_chain_05}
      boxExtents=<10 10 10>
      constraintSetup {angularLimitsMax=<1 45 45>, angularLimitsMin=<-1 -45 -45>, angularTarget=<1 0 0>}
      localJointOffset=<1 0 0>
     }]
    simSpaceSettings {maxAngularAcceleration=75, maxAngularVelocity=75, simSpaceAngularAlpha=1}
   }
   RigidBody!(ComponentPose, Alpha, OverrideWorldGravity:Vector="0,0,-980", ExternalForce) -> (Pose) {
    bClampLinearTranslationLimitToRefPose=true
    bOverrideWorldGravity=true
    overridePhysicsAsset=&/VoidSpray_Axe_Physics
    baseBoneRef {boneName=Melee_TwoHanded_Handle}
    componentAppliedLinearAccClamp=<0 0 0>
    componentLinearAccScale=<0.5 0.5 0.5>
    componentLinearVelScale=<0.5 0.5 0.5>
    simSpaceSettings {
     maxLinearAcceleration=2800
     maxLinearVelocity=700
     worldAlpha=1
     externalAngularVelocity=<1 1 1>
     externalLinearDragV=<1 1 1>
     externalLinearVelocity=<1 1 1>
    }
   }
   n7 = "Transform (Modify) Bone - Bone: base_chain_01"(
    ComponentPose=Local To Component(LocalPose=Use cached pose 'start_pose'())
    Alpha=1
    Translation="0,0,0"
    Rotation="-80.0,0.0,0.0"
    Scale="1,1,1"
   ) {
    rotationMode=BMM_Additive
    translationMode=BMM_Additive
    alphaBoolBlend {blendOption=Linear}
    boneToModify {boneName=base_chain_01}
    rotation {pitch=-80}
   }
   n8 = Component To Local(ComponentPose=@n7)
   n9 = "Transform (Modify) Bone - Bone: base_chain_01"(ComponentPose=Local To Component(LocalPose=Use cached pose 'start_pose'()), Alpha=1, Translation="0,0,0", Rotation="-30,0,0", Scale="1,1,1") {
    rotationMode=BMM_Additive
    translationMode=BMM_Additive
    alphaBoolBlend {blendOption=Linear}
    boneToModify {boneName=base_chain_01}
    rotation {pitch=-30}
   }
   n10 = Component To Local(ComponentPose=@n9)
   n11 = Get bIsOnPawn() -> (bIsOnPawn:bool)
   n12 = Anim Dynamics - Bone: dyn_chain_01(
    ComponentPose=Local To Component(LocalPose=Blend Poses by bool(BlendPose_0=@n8, BlendPose_1=@n10, bActiveValue=@n11))
    Alpha=0.4
    ExternalForce="0,0,0"
   ) {
    angularDampingOverride=1
    angularSpringConstant=20000
    bAngularSpring=true
    bOverrideAngularDamping=true
    bUseGravityOverride=true
    gravityScale=0.35
    simulationSpace=RootRelative
    alphaBoolBlend {blendOption=Linear}
    boundBone {boneName=dyn_chain_01}
    componentLinearAccScale=<1 1 1>
    componentLinearVelScale=<1 1 1>
    gravityOverride {z=-980}
    physicsBodyDefinitions [{
      boundBone {boneName=dyn_chain_01}
      boxExtents=<20 30 30>
      constraintSetup {angularLimitsMax=<20 30 30>, angularLimitsMin=<-20 -5 -30>, angularTarget=<1 0 0>}
      localJointOffset=<4 0 0>
     }]
    simSpaceSettings {simSpaceAngularAlpha=0.3, externalAngularVelocity=<0.2 0.2 0.2>}
   }
   n13 = Anim Dynamics - Bone: dyn_chain_02(ComponentPose=@n12, Alpha=0.4, ExternalForce="0,0,0") {
    angularDampingOverride=1
    angularSpringConstant=20000
    bAngularSpring=true
    bOverrideAngularDamping=true
    bUseGravityOverride=true
    gravityScale=0.35
    simulationSpace=RootRelative
    alphaBoolBlend {blendOption=Linear}
    boundBone {boneName=dyn_chain_02}
    componentLinearAccScale=<1 1 1>
    componentLinearVelScale=<1 1 1>
    gravityOverride {z=-980}
    physicsBodyDefinitions [{
      boundBone {boneName=dyn_chain_02}
      boxExtents=<20 30 30>
      constraintSetup {angularLimitsMax=<20 30 30>, angularLimitsMin=<-20 -5 -30>, angularTarget=<1 0 0>}
      localJointOffset=<4 0 0>
     }]
    simSpaceSettings {simSpaceAngularAlpha=0.3, externalAngularVelocity=<0.2 0.2 0.2>}
   }
   n14 = Anim Dynamics - Bone: dyn_chain_03(ComponentPose=@n13, Alpha=0.4, ExternalForce="0,0,0") {
    angularDampingOverride=1
    angularSpringConstant=20000
    bAngularSpring=true
    bOverrideAngularDamping=true
    bUseGravityOverride=true
    gravityScale=0.35
    simulationSpace=RootRelative
    alphaBoolBlend {blendOption=Linear}
    boundBone {boneName=dyn_chain_03}
    componentLinearAccScale=<1 1 1>
    componentLinearVelScale=<1 1 1>
    gravityOverride {z=-980}
    physicsBodyDefinitions [{
      boundBone {boneName=dyn_chain_03}
      boxExtents=<20 30 30>
      constraintSetup {angularLimitsMax=<20 30 30>, angularLimitsMin=<-20 -5 -30>, angularTarget=<1 0 0>}
      localJointOffset=<4 0 0>
     }]
    simSpaceSettings {simSpaceAngularAlpha=0.3, externalAngularVelocity=<0.2 0.2 0.2>}
   }
   n15 = Anim Dynamics - Bone: dyn_chain_05(ComponentPose=@n14, Alpha=0.4, ExternalForce="0,0,0") {
    angularDampingOverride=1
    angularSpringConstant=20000
    bAngularSpring=true
    bOverrideAngularDamping=true
    bUseGravityOverride=true
    gravityScale=0.35
    simulationSpace=RootRelative
    alphaBoolBlend {blendOption=Linear}
    boundBone {boneName=dyn_chain_05}
    componentLinearAccScale=<1 1 1>
    componentLinearVelScale=<1 1 1>
    gravityOverride {z=-980}
    physicsBodyDefinitions [{
      boundBone {boneName=dyn_chain_05}
      boxExtents=<20 30 30>
      constraintSetup {angularLimitsMax=<20 30 30>, angularLimitsMin=<-20 -5 -30>, angularTarget=<1 0 0>}
      localJointOffset=<4 0 0>
     }]
    simSpaceSettings {simSpaceAngularAlpha=0.3, externalAngularVelocity=<0.2 0.2 0.2>}
   }
   n16 = RigidBody(ComponentPose=Local To Component(LocalPose=Use cached pose 'start_pose'()), OverrideWorldGravity="0,0,-980") {
    bClampLinearTranslationLimitToRefPose=true
    bOverrideWorldGravity=true
    overridePhysicsAsset=&VoidSpray_Axe_Physics
    simulationSpace=BaseBoneSpace
    baseBoneRef {boneName=Melee_TwoHanded_Handle}
    simSpaceSettings {
     dampingAlpha=0
     maxAngularAcceleration=100
     maxAngularVelocity=20
     maxLinearAcceleration=2800
     maxLinearVelocity=700
     velocityScaleZ=0.45
     worldAlpha=1
     externalLinearDragV=<0.2 0.2 0.2>
    }
   }
   Output Pose(Result=Blend Poses by bool(BlendPose_0=Component To Local(ComponentPose=@n15), BlendPose_1=Component To Local(ComponentPose=@n16), bActiveValue=Get bIsInFrontEnd())) {layerGroup=None}
   AnimGraphNode_SubInput_0 = Input Pose()
   Save cached pose 'start_pose'(Pose=@AnimGraphNode_SubInput_0) {cacheName=start_pose}
   Reroute Node(InputPin:Wildcard) -> (OutputPin:Wildcard)
  ]}
 EventGraph {nodes [
   Event OnCharacterCustomizationLoaded()
   > n1 = Branch(Condition=Get bIsInFrontEnd())
   > Set Pickaxe_Rot(Pickaxe_Rot:double=1) -> (Output_Get:double)
   > n2 = Branch(Condition=Get bIsOnPawn())
   > Set Pickaxe_Yval(Pickaxe_Yval:Rotator="-80,0,0") -> (Output_Get:Rotator)
   Set Pickaxe_Yval(execute=@n2.else, Pickaxe_Yval="80,0,0")
   Set Pickaxe_Rot(execute=@n1.else, Pickaxe_Rot=0)
   TryGetPawnOwner()
   AND Boolean(A=false, B=false)
   Event BlueprintUpdateAnimation()
  ]}
]
variables [
 NewVar {type=boolean}
 Pickaxe_Rot {type=number}
 Pickaxe_Yval {type=Rotator}
]
